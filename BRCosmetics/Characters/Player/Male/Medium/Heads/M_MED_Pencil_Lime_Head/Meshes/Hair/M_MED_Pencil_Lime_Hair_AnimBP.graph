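# Animation Blueprint for the Pencil Lime male head's hair, adding secondary physics motion to bones dyn_hair_A through S.
# Input Pose feeds serial root-relative Anim Dynamics nodes with angular springs and limits, dyn_hair_N appearing twice; Copy Pose From Mesh is unconnected.
# Each update sets GravityOverride through a deprecated gravity-state function, but the nodes show fixed gravity overrides and no link to that variable.
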AnimBlueprint /BRCosmetics/Characters/Player/Male/Medium/Heads/M_MED_Pencil_Lime_Head/Meshes/Hair/M_MED_Pencil_Lime_Hair_AnimBP
target_skeleton=/BRCosmetics/Characters/Player/Common/Fortnite_Base_Head/Pencil/Fortnite_Base_Head_Pencil_Skeleton
graphs [
 AnimGraph {nodes [
   Copy Pose From Mesh() {bCopyCurves=true, bUseAttachedParent=true}
   n1 = Anim Dynamics - Bone: dyn_hair_A(ComponentPose=Local To Component(LocalPose=Input Pose()), Alpha=0.5, ExternalForce="0,0,0") {
    angularDampingOverride=1
    angularSpringConstant=4000
    bAngularSpring=true
    bGravityOverrideInSimSpace=true
    bOverrideAngularDamping=true
    bUseGravityOverride=true
    gravityScale=0.35
    simulationSpace=RootRelative
    alphaBoolBlend {blendOption=Linear}
    alphaScaleBiasClamp {scale=0.8}
    boundBone {boneName=dyn_hair_A}
    componentAppliedLinearAccClamp=<2500 2500 2500>
    componentLinearAccScale=<0.48 0.48 1>
    componentLinearVelScale=<0.24 0.24 0.5>
    gravityOverride {z=-980}
    physicsBodyDefinitions [{
      sphereCollisionRadius=0
      boundBone {boneName=dyn_hair_A}
      boxExtents=<10 20 20>
      constraintSetup {angularLimitsMax=<10 3 45>, angularLimitsMin=<-10 -60 -45>, angularTarget=<1 0 0>}
      localJointOffset=<1 0 0>
     }]
   }
   n2 = Anim Dynamics - Bone: dyn_hair_B(ComponentPose=@n1, Alpha=0.5, ExternalForce="0,0,0") {
    angularDampingOverride=1
    angularSpringConstant=4000
    bAngularSpring=true
    bGravityOverrideInSimSpace=true
    bOverrideAngularDamping=true
    bUseGravityOverride=true
    gravityScale=0.35
    simulationSpace=RootRelative
    alphaBoolBlend {blendOption=Linear}
    alphaScaleBiasClamp {scale=0.8}
    boundBone {boneName=dyn_hair_B}
    componentAppliedLinearAccClamp=<2500 2500 2500>
    componentLinearAccScale=<0.48 0.48 1>
    componentLinearVelScale=<0.24 0.24 0.5>
    gravityOverride {z=-980}
    physicsBodyDefinitions [{
      sphereCollisionRadius=0
      boundBone {boneName=dyn_hair_B}
      boxExtents=<10 20 20>
      constraintSetup {angularLimitsMax=<10 3 45>, angularLimitsMin=<-10 -60 -45>, angularTarget=<1 0 0>}
      localJointOffset=<1 0 0>
     }]
   }
   n3 = Anim Dynamics - Bone: dyn_hair_C(ComponentPose=@n2, Alpha=1.3, ExternalForce="0,0,0") {
    angularDampingOverride=1
    angularSpringConstant=4000
    bAngularSpring=true
    bGravityOverrideInSimSpace=true
    bOverrideAngularDamping=true
    bUseGravityOverride=true
    gravityScale=0.35
    simulationSpace=RootRelative
    alphaBoolBlend {blendOption=Linear}
    alphaScaleBiasClamp {scale=0.8}
    boundBone {boneName=dyn_hair_C}
    componentAppliedLinearAccClamp=<2500 2500 2500>
    componentLinearAccScale=<0.48 0.48 1>
    componentLinearVelScale=<0.24 0.24 0.5>
    gravityOverride {z=-980}
    physicsBodyDefinitions [{
      sphereCollisionRadius=0
      boundBone {boneName=dyn_hair_C}
      boxExtents=<10 20 20>
      constraintSetup {angularLimitsMax=<10 3 45>, angularLimitsMin=<-10 -60 -45>, angularTarget=<1 0 0>}
      localJointOffset=<1 0 0>
     }]
   }
   n4 = Anim Dynamics - Bone: dyn_hair_D(ComponentPose=@n3, Alpha=1.3, ExternalForce="0,0,0") {
    angularDampingOverride=1
    angularSpringConstant=4000
    bAngularSpring=true
    bGravityOverrideInSimSpace=true
    bOverrideAngularDamping=true
    bUseGravityOverride=true
    gravityScale=0.35
    simulationSpace=RootRelative
    alphaBoolBlend {blendOption=Linear}
    alphaScaleBiasClamp {scale=0.8}
    boundBone {boneName=dyn_hair_D}
    componentAppliedLinearAccClamp=<2500 2500 2500>
    componentLinearAccScale=<0.48 0.48 1>
    componentLinearVelScale=<0.24 0.24 0.5>
    gravityOverride {z=-980}
    physicsBodyDefinitions [{
      sphereCollisionRadius=0
      boundBone {boneName=dyn_hair_D}
      boxExtents=<10 20 20>
      constraintSetup {angularLimitsMax=<10 3 45>, angularLimitsMin=<-10 -60 -45>, angularTarget=<1 0 0>}
      localJointOffset=<1 0 0>
     }]
   }
   n5 = Anim Dynamics - Bone: dyn_hair_E(ComponentPose=@n4, Alpha=1, ExternalForce="0,0,0") {
    angularDampingOverride=1
    angularSpringConstant=4000
    bAngularSpring=true
    bGravityOverrideInSimSpace=true
    bOverrideAngularDamping=true
    bUseGravityOverride=true
    gravityScale=0.35
    simulationSpace=RootRelative
    alphaBoolBlend {blendOption=Linear}
    alphaScaleBiasClamp {scale=0.8}
    boundBone {boneName=dyn_hair_E}
    componentAppliedLinearAccClamp=<2500 2500 2500>
    componentLinearAccScale=<0.48 0.48 1>
    componentLinearVelScale=<0.24 0.24 0.5>
    gravityOverride {z=-980}
    physicsBodyDefinitions [{
      sphereCollisionRadius=0
      boundBone {boneName=dyn_hair_E}
      boxExtents=<10 20 20>
      constraintSetup {angularLimitsMax=<10 3 45>, angularLimitsMin=<-10 -60 -45>, angularTarget=<1 0 0>}
      localJointOffset=<1 0 0>
     }]
   }
   n6 = Anim Dynamics - Bone: dyn_hair_F(ComponentPose=@n5, Alpha=1.3, ExternalForce="0,0,0") {
    angularDampingOverride=1
    angularSpringConstant=4000
    bAngularSpring=true
    bGravityOverrideInSimSpace=true
    bOverrideAngularDamping=true
    bUseGravityOverride=true
    gravityScale=0.35
    simulationSpace=RootRelative
    alphaBoolBlend {blendOption=Linear}
    alphaScaleBiasClamp {scale=0.8}
    boundBone {boneName=dyn_hair_F}
    componentAppliedLinearAccClamp=<2500 2500 2500>
    componentLinearAccScale=<0.48 0.48 1>
    componentLinearVelScale=<0.24 0.24 0.5>
    gravityOverride {z=-980}
    physicsBodyDefinitions [{
      sphereCollisionRadius=0
      boundBone {boneName=dyn_hair_F}
      boxExtents=<10 20 20>
      constraintSetup {angularLimitsMax=<10 3 45>, angularLimitsMin=<-10 -60 -45>, angularTarget=<1 0 0>}
      localJointOffset=<1 0 0>
     }]
   }
   n7 = Anim Dynamics - Bone: dyn_hair_G(ComponentPose=@n6, Alpha=1.3, ExternalForce="0,0,0") {
    angularDampingOverride=1
    angularSpringConstant=4000
    bAngularSpring=true
    bGravityOverrideInSimSpace=true
    bOverrideAngularDamping=true
    bUseGravityOverride=true
    gravityScale=0.35
    simulationSpace=RootRelative
    alphaBoolBlend {blendOption=Linear}
    alphaScaleBiasClamp {scale=0.8}
    boundBone {boneName=dyn_hair_G}
    componentAppliedLinearAccClamp=<2500 2500 2500>
    componentLinearAccScale=<0.48 0.48 1>
    componentLinearVelScale=<0.24 0.24 0.5>
    gravityOverride {z=-980}
    physicsBodyDefinitions [{
      sphereCollisionRadius=0
      boundBone {boneName=dyn_hair_G}
      boxExtents=<10 20 20>
      constraintSetup {angularLimitsMax=<10 3 45>, angularLimitsMin=<-10 -60 -45>, angularTarget=<1 0 0>}
      localJointOffset=<1 0 0>
     }]
   }
   n8 = Anim Dynamics - Bone: dyn_hair_H(ComponentPose=@n7, Alpha=1, ExternalForce="0,0,0") {
    angularDampingOverride=1
    angularSpringConstant=4000
    bAngularSpring=true
    bGravityOverrideInSimSpace=true
    bOverrideAngularDamping=true
    bUseGravityOverride=true
    gravityScale=0.35
    simulationSpace=RootRelative
    alphaBoolBlend {blendOption=Linear}
    alphaScaleBiasClamp {scale=0.8}
    boundBone {boneName=dyn_hair_H}
    componentAppliedLinearAccClamp=<2500 2500 2500>
    componentLinearAccScale=<0.48 0.48 1>
    componentLinearVelScale=<0.24 0.24 0.5>
    gravityOverride {z=-980}
    physicsBodyDefinitions [{
      sphereCollisionRadius=0
      boundBone {boneName=dyn_hair_H}
      boxExtents=<10 20 20>
      constraintSetup {angularLimitsMax=<10 3 45>, angularLimitsMin=<-10 -60 -45>, angularTarget=<1 0 0>}
      localJointOffset=<1 0 0>
     }]
   }
   n9 = Anim Dynamics - Bone: dyn_hair_I(ComponentPose=@n8, Alpha=1.3, ExternalForce="0,0,0") {
    angularDampingOverride=1
    angularSpringConstant=4000
    bAngularSpring=true
    bGravityOverrideInSimSpace=true
    bOverrideAngularDamping=true
    bUseGravityOverride=true
    gravityScale=0.35
    simulationSpace=RootRelative
    alphaBoolBlend {blendOption=Linear}
    alphaScaleBiasClamp {scale=0.8}
    boundBone {boneName=dyn_hair_I}
    componentAppliedLinearAccClamp=<2500 2500 2500>
    componentLinearAccScale=<0.48 0.48 1>
    componentLinearVelScale=<0.24 0.24 0.5>
    gravityOverride {z=-980}
    physicsBodyDefinitions [{
      sphereCollisionRadius=0
      boundBone {boneName=dyn_hair_I}
      boxExtents=<10 20 20>
      constraintSetup {angularLimitsMax=<10 3 45>, angularLimitsMin=<-10 -60 -45>, angularTarget=<1 0 0>}
      localJointOffset=<1 0 0>
     }]
   }
   n10 = Anim Dynamics - Bone: dyn_hair_J(ComponentPose=@n9, Alpha=1, ExternalForce="0,0,0") {
    angularDampingOverride=1
    angularSpringConstant=4000
    bAngularSpring=true
    bGravityOverrideInSimSpace=true
    bOverrideAngularDamping=true
    bUseGravityOverride=true
    gravityScale=0.35
    simulationSpace=RootRelative
    alphaBoolBlend {blendOption=Linear}
    alphaScaleBiasClamp {scale=0.8}
    boundBone {boneName=dyn_hair_J}
    componentAppliedLinearAccClamp=<2500 2500 2500>
    componentLinearAccScale=<0.48 0.48 1>
    componentLinearVelScale=<0.24 0.24 0.5>
    gravityOverride {z=-980}
    physicsBodyDefinitions [{
      sphereCollisionRadius=0
      boundBone {boneName=dyn_hair_J}
      boxExtents=<10 20 20>
      constraintSetup {angularLimitsMax=<5 3 10>, angularLimitsMin=<-5 -60 -10>, angularTarget=<1 0 0>}
      localJointOffset=<1 0 0>
     }]
   }
   n11 = Anim Dynamics - Bone: dyn_hair_K(ComponentPose=@n10, Alpha=1, ExternalForce="0,0,0") {
    angularDampingOverride=1
    angularSpringConstant=4000
    bAngularSpring=true
    bGravityOverrideInSimSpace=true
    bOverrideAngularDamping=true
    bUseGravityOverride=true
    gravityScale=0.35
    simulationSpace=RootRelative
    alphaBoolBlend {blendOption=Linear}
    alphaScaleBiasClamp {scale=0.8}
    boundBone {boneName=dyn_hair_K}
    componentAppliedLinearAccClamp=<2500 2500 2500>
    componentLinearAccScale=<0.48 0.48 1>
    componentLinearVelScale=<0.24 0.24 0.5>
    gravityOverride {z=-980}
    physicsBodyDefinitions [{
      sphereCollisionRadius=0
      boundBone {boneName=dyn_hair_K}
      boxExtents=<10 20 20>
      constraintSetup {angularLimitsMax=<10 3 45>, angularLimitsMin=<-10 -60 -45>, angularTarget=<1 0 0>}
      localJointOffset=<1 0 0>
     }]
   }
   n12 = Anim Dynamics - Bone: dyn_hair_L(ComponentPose=@n11, Alpha=1, ExternalForce="0,0,0") {
    angularDampingOverride=1
    angularSpringConstant=4000
    bAngularSpring=true
    bGravityOverrideInSimSpace=true
    bOverrideAngularDamping=true
    bUseGravityOverride=true
    gravityScale=0.35
    simulationSpace=RootRelative
    alphaBoolBlend {blendOption=Linear}
    alphaScaleBiasClamp {scale=0.8}
    boundBone {boneName=dyn_hair_L}
    componentAppliedLinearAccClamp=<2500 2500 2500>
    componentLinearAccScale=<0.48 0.48 1>
    componentLinearVelScale=<0.24 0.24 0.5>
    gravityOverride {z=-980}
    physicsBodyDefinitions [{
      sphereCollisionRadius=0
      boundBone {boneName=dyn_hair_L}
      boxExtents=<10 20 20>
      constraintSetup {angularLimitsMax=<10 3 45>, angularLimitsMin=<-10 -60 -45>, angularTarget=<1 0 0>}
      localJointOffset=<1 0 0>
     }]
   }
   n13 = Anim Dynamics - Bone: dyn_hair_M(ComponentPose=@n12, Alpha=1.3, ExternalForce="0,0,0") {
    angularDampingOverride=1
    angularSpringConstant=4000
    bAngularSpring=true
    bGravityOverrideInSimSpace=true
    bOverrideAngularDamping=true
    bUseGravityOverride=true
    gravityScale=0.35
    simulationSpace=RootRelative
    alphaBoolBlend {blendOption=Linear}
    alphaScaleBiasClamp {scale=0.8}
    boundBone {boneName=dyn_hair_M}
    componentAppliedLinearAccClamp=<2500 2500 2500>
    componentLinearAccScale=<0.48 0.48 1>
    componentLinearVelScale=<0.24 0.24 0.5>
    gravityOverride {z=-980}
    physicsBodyDefinitions [{
      sphereCollisionRadius=0
      boundBone {boneName=dyn_hair_M}
      boxExtents=<10 20 20>
      constraintSetup {angularLimitsMax=<10 3 45>, angularLimitsMin=<-10 -60 -45>, angularTarget=<1 0 0>}
      localJointOffset=<1 0 0>
     }]
   }
   n14 = Anim Dynamics - Bone: dyn_hair_N(ComponentPose=@n13, Alpha=1.3, ExternalForce="0,0,0") {
    angularDampingOverride=1
    angularSpringConstant=4000
    bAngularSpring=true
    bGravityOverrideInSimSpace=true
    bOverrideAngularDamping=true
    bUseGravityOverride=true
    gravityScale=0.35
    simulationSpace=RootRelative
    alphaBoolBlend {blendOption=Linear}
    alphaScaleBiasClamp {scale=0.8}
    boundBone {boneName=dyn_hair_N}
    componentAppliedLinearAccClamp=<2500 2500 2500>
    componentLinearAccScale=<0.48 0.48 1>
    componentLinearVelScale=<0.24 0.24 0.5>
    gravityOverride {z=-980}
    physicsBodyDefinitions [{
      sphereCollisionRadius=0
      boundBone {boneName=dyn_hair_N}
      boxExtents=<10 20 20>
      constraintSetup {angularLimitsMax=<10 3 45>, angularLimitsMin=<-10 -60 -45>, angularTarget=<1 0 0>}
      localJointOffset=<1 0 0>
     }]
   }
   n15 = Anim Dynamics - Bone: dyn_hair_N(ComponentPose=@n14, Alpha=1.3, ExternalForce="0,0,0") {
    angularDampingOverride=1
    angularSpringConstant=4000
    bAngularSpring=true
    bGravityOverrideInSimSpace=true
    bOverrideAngularDamping=true
    bUseGravityOverride=true
    gravityScale=0.35
    simulationSpace=RootRelative
    alphaBoolBlend {blendOption=Linear}
    alphaScaleBiasClamp {scale=0.8}
    boundBone {boneName=dyn_hair_N}
    componentAppliedLinearAccClamp=<2500 2500 2500>
    componentLinearAccScale=<0.48 0.48 1>
    componentLinearVelScale=<0.24 0.24 0.5>
    gravityOverride {z=-980}
    physicsBodyDefinitions [{
      sphereCollisionRadius=0
      boundBone {boneName=dyn_hair_N}
      boxExtents=<10 20 20>
      constraintSetup {angularLimitsMax=<10 3 45>, angularLimitsMin=<-10 -60 -45>, angularTarget=<1 0 0>}
      localJointOffset=<1 0 0>
     }]
   }
   n16 = Anim Dynamics - Bone: dyn_hair_O(ComponentPose=@n15, Alpha=1.1, ExternalForce="0,0,0") {
    angularDampingOverride=1
    angularSpringConstant=4000
    bAngularSpring=true
    bGravityOverrideInSimSpace=true
    bOverrideAngularDamping=true
    bUseGravityOverride=true
    gravityScale=0.35
    simulationSpace=RootRelative
    alphaBoolBlend {blendOption=Linear}
    alphaScaleBiasClamp {scale=0.8}
    boundBone {boneName=dyn_hair_O}
    componentAppliedLinearAccClamp=<2500 2500 2500>
    componentLinearAccScale=<0.48 0.48 1>
    componentLinearVelScale=<0.24 0.24 0.5>
    gravityOverride {z=-980}
    physicsBodyDefinitions [{
      sphereCollisionRadius=0
      boundBone {boneName=dyn_hair_O}
      boxExtents=<10 20 20>
      constraintSetup {angularLimitsMax=<10 3 45>, angularLimitsMin=<-10 -60 -45>, angularTarget=<1 0 0>}
      localJointOffset=<1 0 0>
     }]
   }
   n17 = Anim Dynamics - Bone: dyn_hair_P(ComponentPose=@n16, Alpha=1, ExternalForce="0,0,0") {
    angularDampingOverride=1
    angularSpringConstant=4000
    bAngularSpring=true
    bGravityOverrideInSimSpace=true
    bOverrideAngularDamping=true
    bUseGravityOverride=true
    gravityScale=0.35
    simulationSpace=RootRelative
    alphaBoolBlend {blendOption=Linear}
    alphaScaleBiasClamp {scale=0.8}
    boundBone {boneName=dyn_hair_P}
    componentAppliedLinearAccClamp=<2500 2500 2500>
    componentLinearAccScale=<0.48 0.48 1>
    componentLinearVelScale=<0.24 0.24 0.5>
    gravityOverride {z=-980}
    physicsBodyDefinitions [{
      sphereCollisionRadius=0
      boundBone {boneName=dyn_hair_P}
      boxExtents=<10 20 20>
      constraintSetup {angularLimitsMax=<10 3 45>, angularLimitsMin=<-10 -60 -45>, angularTarget=<1 0 0>}
      localJointOffset=<1 0 0>
     }]
   }
   n18 = Anim Dynamics - Bone: dyn_hair_Q(ComponentPose=@n17, Alpha=1, ExternalForce="0,0,0") {
    angularDampingOverride=1
    angularSpringConstant=4000
    bAngularSpring=true
    bGravityOverrideInSimSpace=true
    bOverrideAngularDamping=true
    bUseGravityOverride=true
    gravityScale=0.35
    simulationSpace=RootRelative
    alphaBoolBlend {blendOption=Linear}
    alphaScaleBiasClamp {scale=0.8}
    boundBone {boneName=dyn_hair_Q}
    componentAppliedLinearAccClamp=<2500 2500 2500>
    componentLinearAccScale=<0.48 0.48 1>
    componentLinearVelScale=<0.24 0.24 0.5>
    gravityOverride {z=-980}
    physicsBodyDefinitions [{
      sphereCollisionRadius=0
      boundBone {boneName=dyn_hair_Q}
      boxExtents=<10 20 20>
      constraintSetup {angularLimitsMax=<10 3 45>, angularLimitsMin=<-10 -60 -45>, angularTarget=<1 0 0>}
      localJointOffset=<1 0 0>
     }]
   }
   n19 = Anim Dynamics - Bone: dyn_hair_R(ComponentPose=@n18, Alpha=1, ExternalForce="0,0,0") {
    angularDampingOverride=1
    angularSpringConstant=4000
    bAngularSpring=true
    bGravityOverrideInSimSpace=true
    bOverrideAngularDamping=true
    bUseGravityOverride=true
    gravityScale=0.35
    simulationSpace=RootRelative
    alphaBoolBlend {blendOption=Linear}
    alphaScaleBiasClamp {scale=0.8}
    boundBone {boneName=dyn_hair_R}
    componentAppliedLinearAccClamp=<2500 2500 2500>
    componentLinearAccScale=<0.48 0.48 1>
    componentLinearVelScale=<0.24 0.24 0.5>
    gravityOverride {z=-980}
    physicsBodyDefinitions [{
      sphereCollisionRadius=0
      boundBone {boneName=dyn_hair_R}
      boxExtents=<10 20 20>
      constraintSetup {angularLimitsMax=<10 3 30>, angularLimitsMin=<-10 -40 -30>, angularTarget=<1 0 0>}
      localJointOffset=<1 0 0>
     }]
   }
   n20 = Anim Dynamics - Bone: dyn_hair_S(ComponentPose=@n19, Alpha=1, ExternalForce="0,0,0") {
    angularDampingOverride=1
    angularSpringConstant=4000
    bAngularSpring=true
    bGravityOverrideInSimSpace=true
    bOverrideAngularDamping=true
    bUseGravityOverride=true
    gravityScale=0.35
    simulationSpace=RootRelative
    alphaBoolBlend {blendOption=Linear}
    alphaScaleBiasClamp {scale=0.8}
    boundBone {boneName=dyn_hair_S}
    componentAppliedLinearAccClamp=<2500 2500 2500>
    componentLinearAccScale=<0.48 0.48 1>
    componentLinearVelScale=<0.24 0.24 0.5>
    gravityOverride {z=-980}
    physicsBodyDefinitions [{
      sphereCollisionRadius=0
      boundBone {boneName=dyn_hair_S}
      boxExtents=<10 20 20>
      constraintSetup {angularLimitsMax=<10 3 30>, angularLimitsMin=<-10 -40 -30>, angularTarget=<1 0 0>}
      localJointOffset=<1 0 0>
     }]
   }
   Output Pose(Result=Component To Local(ComponentPose=@n20)) {layerGroup=None}
  ]}
 EventGraph {nodes [
   Event BlueprintUpdateAnimation()
   > n1 = Gravity Override State Function DEPRECATED(CharacterPartAnimInstance=Self-Reference(), GameTimeInSeconds=GetGameTimeInSeconds(), GravityOverrideParams=Get Gravity Override Params())
   > Set GravityOverride(GravityOverride=@n1)
  ]}
]
variables [
 Gravity Override Params {type=GravityOverrideParamsStruct}
 GravityOverride {type=Vector}
]
